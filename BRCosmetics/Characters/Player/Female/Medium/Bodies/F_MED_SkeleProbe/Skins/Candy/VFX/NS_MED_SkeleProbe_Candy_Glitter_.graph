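# Niagara system that adds tinted glitter sprites to the body and face accessory of the female SkeleProbe Candy skin.
# Seven body-region emitters share system spawn and size settings, pinning particles to skeletal mesh bones or sockets with user-array offsets.
# Paired limbs alternate by particle partition, negating offsets to mirror one side.
# Short-lived sprites spin, grow and fade, with LOD-distance size scaling and time-of-day brightness applied to the user color.
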NiagaraSystem /BRCosmetics/Characters/Player/Female/Medium/Bodies/F_MED_SkeleProbe/Skins/Candy/VFX/NS_MED_SkeleProbe_Candy_Glitter_
dependencies {
 usedDataInterfaces [&NiagaraDataInterfaceArrayInt32, &NiagaraDataInterfaceCurve]
 usedDynamicInputs []
 usedModules [
  &SystemState
  &self:SystemSpawnScript.SystemScriptSource.SystemScriptGraph.NiagaraNodeAssignment_1.SetVariables_#1
 ]
}
emitters {
 Arms {
  emitterUpdateScript [
   &EmitterState()
   &SpawnRate(
    Spawn Probability=$System.SpawnProbability
    SpawnRate=&Multiply_Float(A=$System.SpawnRate, B=1)
    Use Spawn Probability=true
   )
  ]
  particleSpawnScript [
   &PartitionParticles(Particles Per Partition=1, Partition Mode:ENiagara_PartitionMode=::NewEnumerator1)
   &InitializeParticle(
    Color Mode=::NewEnumerator0
    Lifetime=&RandomRangeFloat(Minimum=0.3, Maximum=0.4)
    Sprite Rotation Mode=::NewEnumerator0
    Sprite Size Mode=::NewEnumerator1
    Uniform Sprite Size Max=$System.MaxSize
    Uniform Sprite Size Min=$System.MinSize
    UsePositionOffset=false
    Write Material Random=true
   )
   &CameraOffset(Camera Offset Amount=$System.CameraOffset)
  ]
  particleUpdateScript [
   &ParticleState()
   &SkeletalMeshLocation(
    Bone / Socket Index=&SelectIntFromArray(Direct Array Index=$Particles.PartitionParticles.Partition, Int Selection Array=DataInterface() {propertyValues {IntData [4,5]}})
    Bone Sampling Mode=::NewEnumerator4
    MaskSpawnGroup=false
    Sampled Position Offset=&Multiply_Vector(
     A=&SelectVectorFromArray(Array Sampling Mode=::NewEnumerator0, Direct Array Index=?, Vector Selection Array=$User.Array_Arm, Random Seed=0)
     B=&MaskVectorBySpawnGroup(Masked Vector=<1 1 1>, Passthrough Vector=<-1 -1 -1>, Spawn Group Mask=$Particles.PartitionParticles.Partition)
    )
    Skeletal Mesh=$User.Skeletal Mesh Body
   ) {module=SkeletalMeshLocation004}
   &SpriteRotationRate(Rotation Rate=250)
   &ScaleSpriteSize(
    Scale Sprite Size Mode=::NewEnumerator0
    Uniform Scale Factor=&FloatFromCurve(
     FloatCurve=DataInterface() {propertyValues {Curve {keys [{arriveTangent=2.9879, interpMode=RCIM_Cubic, leaveTangent=2.9879, tangentMode=RCTM_User, time=0, value=0}, <at 1 1 RCIM_Cubic>]}}}
     Scale Curve=&FN_LOD_Distance_Remap(Start Distance=100, End Distance=500, Minimum Output=0.75, Maximum Output=1.25, Clamp Output=true)
    )
   )
   &ScaleColor(
    Scale Alpha=&FloatFromCurve(FloatCurve=DataInterface() {propertyValues {Curve {keys [
         {arriveTangent=9.0725, interpMode=RCIM_Cubic, leaveTangent=9.0725, tangentMode=RCTM_User, time=0, value=0}
         {arriveTangent=-2.5249, interpMode=RCIM_Cubic, leaveTangent=-2.5249, tangentMode=RCTM_User, time=0.1408, value=0.9847}
         {arriveTangent=-0.0405, interpMode=RCIM_Cubic, leaveTangent=-0.0405, tangentMode=RCTM_User, time=1, value=0}
        ]}}})
    Scale RGB=&Multiply_VectorByFloat(
     Vector=&MakeVectorFromLinearColorRGB(Color=$User.Linear Color)
     Float=&FNTimeOfDayScalar(Night Scale=0.2, Morning Scale=0.5, Day Scale=0.5, Evening Scale=0.5)
    )
    ScaleA=true
    ScaleRGB=true
   )
  ]
  rendererClasses [&NiagaraSpriteRendererProperties]
  renderers [{rendererClass=&NiagaraSpriteRendererProperties}]
 }
 Belly_Back {
  emitterUpdateScript [
   &EmitterState()
   &SpawnRate(
    Spawn Probability=$System.SpawnProbability
    SpawnRate=&Multiply_Float(A=$System.SpawnRate, B=0.6)
    Use Spawn Probability=true
   )
  ]
  particleSpawnScript [
   &PartitionParticles(Particles Per Partition=1, Partition Mode:ENiagara_PartitionMode=::NewEnumerator1)
   &InitializeParticle(
    Color Mode=::NewEnumerator0
    Lifetime=&RandomRangeFloat(Minimum=0.3, Maximum=0.4)
    Sprite Rotation Mode=::NewEnumerator0
    Sprite Size Mode=::NewEnumerator1
    Uniform Sprite Size Max=$System.MaxSize
    Uniform Sprite Size Min=$System.MinSize
    UsePositionOffset=false
    Write Material Random=true
   )
   &CameraOffset(Camera Offset Amount=$System.CameraOffset)
  ]
  particleUpdateScript [
   &ParticleState()
   &SkeletalMeshLocation(
    Bone / Socket Index=&SelectIntFromArray(Direct Array Index=$Particles.PartitionParticles.Partition, Int Selection Array=DataInterface() {propertyValues {IntData [8]}})
    Bone Sampling Mode=::NewEnumerator4
    MaskSpawnGroup=false
    Sampled Position Offset=&SelectVectorFromArray(Array Sampling Mode=::NewEnumerator0, Direct Array Index=?, Vector Selection Array=$User.Array_Belly, Random Seed=0)
    Skeletal Mesh=$User.Skeletal Mesh Body
   ) {module=SkeletalMeshLocation004}
   &SpriteRotationRate(Rotation Rate=250)
   &ScaleSpriteSize(
    Scale Sprite Size Mode=::NewEnumerator0
    Uniform Scale Factor=&FloatFromCurve(
     FloatCurve=DataInterface() {propertyValues {Curve {keys [{arriveTangent=2.9879, interpMode=RCIM_Cubic, leaveTangent=2.9879, tangentMode=RCTM_User, time=0, value=0}, <at 1 1 RCIM_Cubic>]}}}
     Scale Curve=&FN_LOD_Distance_Remap(Start Distance=100, End Distance=500, Minimum Output=0.75, Maximum Output=1.25, Clamp Output=true)
    )
   )
   &ScaleColor(
    Scale Alpha=&FloatFromCurve(FloatCurve=DataInterface() {propertyValues {Curve {keys [
         {arriveTangent=9.0725, interpMode=RCIM_Cubic, leaveTangent=9.0725, tangentMode=RCTM_User, time=0, value=0}
         {arriveTangent=-2.5249, interpMode=RCIM_Cubic, leaveTangent=-2.5249, tangentMode=RCTM_User, time=0.1408, value=0.9847}
         {arriveTangent=-0.0405, interpMode=RCIM_Cubic, leaveTangent=-0.0405, tangentMode=RCTM_User, time=1, value=0}
        ]}}})
    Scale RGB=&Multiply_VectorByFloat(
     Vector=&MakeVectorFromLinearColorRGB(Color=$User.Linear Color)
     Float=&FNTimeOfDayScalar(Night Scale=0.2, Morning Scale=0.5, Day Scale=0.5, Evening Scale=0.5)
    )
    ScaleA=true
    ScaleRGB=true
   )
  ]
  rendererClasses [&NiagaraSpriteRendererProperties]
  renderers [{rendererClass=&NiagaraSpriteRendererProperties}]
 }
 Calfs {
  emitterUpdateScript [
   &EmitterState()
   &SpawnRate(
    Spawn Probability=$System.SpawnProbability
    SpawnRate=&Multiply_Float(A=$System.SpawnRate, B=1)
    Use Spawn Probability=true
   )
  ]
  particleSpawnScript [
   &PartitionParticles(Particles Per Partition=1, Partition Mode:ENiagara_PartitionMode=::NewEnumerator1)
   &InitializeParticle(
    Color Mode=::NewEnumerator0
    Lifetime=&RandomRangeFloat(Minimum=0.3, Maximum=0.4)
    Sprite Rotation Mode=::NewEnumerator0
    Sprite Size Mode=::NewEnumerator1
    Uniform Sprite Size Max=$System.MaxSize
    Uniform Sprite Size Min=$System.MinSize
    UsePositionOffset=false
    Write Material Random=true
   )
   &CameraOffset(Camera Offset Amount=$System.CameraOffset)
  ]
  particleUpdateScript [
   &ParticleState()
   &SkeletalMeshLocation(
    Bone / Socket Index=&SelectIntFromArray(Direct Array Index=$Particles.PartitionParticles.Partition, Int Selection Array=DataInterface() {propertyValues {IntData [0,1]}})
    Bone Sampling Mode=::NewEnumerator4
    MaskSpawnGroup=false
    Sampled Position Offset=&Multiply_Vector(
     A=&SelectVectorFromArray(Array Sampling Mode=::NewEnumerator0, Direct Array Index=?, Vector Selection Array=$User.Array_Calf, Random Seed=0)
     B=&MaskVectorBySpawnGroup(Masked Vector=<1 1 1>, Passthrough Vector=<-1 -1 -1>, Spawn Group Mask=$Particles.PartitionParticles.Partition)
    )
    Skeletal Mesh=$User.Skeletal Mesh Body
   ) {module=SkeletalMeshLocation004}
   &SpriteRotationRate(Rotation Rate=250)
   &ScaleSpriteSize(
    Scale Sprite Size Mode=::NewEnumerator0
    Uniform Scale Factor=&FloatFromCurve(
     FloatCurve=DataInterface() {propertyValues {Curve {keys [{arriveTangent=2.9879, interpMode=RCIM_Cubic, leaveTangent=2.9879, tangentMode=RCTM_User, time=0, value=0}, <at 1 1 RCIM_Cubic>]}}}
     Scale Curve=&FN_LOD_Distance_Remap(Start Distance=100, End Distance=500, Minimum Output=0.75, Maximum Output=1.25, Clamp Output=true)
    )
   )
   &ScaleColor(
    Scale Alpha=&FloatFromCurve(FloatCurve=DataInterface() {propertyValues {Curve {keys [
         {arriveTangent=9.0725, interpMode=RCIM_Cubic, leaveTangent=9.0725, tangentMode=RCTM_User, time=0, value=0}
         {arriveTangent=-2.5249, interpMode=RCIM_Cubic, leaveTangent=-2.5249, tangentMode=RCTM_User, time=0.1408, value=0.9847}
         {arriveTangent=-0.0405, interpMode=RCIM_Cubic, leaveTangent=-0.0405, tangentMode=RCTM_User, time=1, value=0}
        ]}}})
    Scale RGB=&Multiply_VectorByFloat(
     Vector=&MakeVectorFromLinearColorRGB(Color=$User.Linear Color)
     Float=&FNTimeOfDayScalar(Night Scale=0.2, Morning Scale=0.5, Day Scale=0.5, Evening Scale=0.5)
    )
    ScaleA=true
    ScaleRGB=true
   )
  ]
  rendererClasses [&NiagaraSpriteRendererProperties]
  renderers [{rendererClass=&NiagaraSpriteRendererProperties}]
 }
 Chest {
  emitterUpdateScript [
   &EmitterState()
   &SpawnRate(
    Spawn Probability=$System.SpawnProbability
    SpawnRate=&Multiply_Float(A=$System.SpawnRate, B=1)
    Use Spawn Probability=true
   )
  ]
  particleSpawnScript [
   &PartitionParticles(Particles Per Partition=1, Partition Mode:ENiagara_PartitionMode=::NewEnumerator1)
   &InitializeParticle(
    Color Mode=::NewEnumerator0
    Lifetime=&RandomRangeFloat(Minimum=0.3, Maximum=0.4)
    Sprite Rotation Mode=::NewEnumerator0
    Sprite Size Mode=::NewEnumerator1
    Uniform Sprite Size Max=$System.MaxSize
    Uniform Sprite Size Min=$System.MinSize
    UsePositionOffset=false
    Write Material Random=true
   )
   &CameraOffset(Camera Offset Amount=$System.CameraOffset)
  ]
  particleUpdateScript [
   &ParticleState()
   &SkeletalMeshLocation(
    Bone / Socket Index=&SelectIntFromArray(Direct Array Index=$Particles.PartitionParticles.Partition, Int Selection Array=DataInterface() {propertyValues {IntData [9]}})
    Bone Sampling Mode=::NewEnumerator4
    MaskSpawnGroup=false
    Sampled Position Offset=&SelectVectorFromArray(Array Sampling Mode=::NewEnumerator0, Direct Array Index=?, Vector Selection Array=$User.Array_Chest, Random Seed=0)
    Skeletal Mesh=$User.Skeletal Mesh Body
   ) {module=SkeletalMeshLocation004}
   &SpriteRotationRate(Rotation Rate=250)
   &ScaleSpriteSize(
    Scale Sprite Size Mode=::NewEnumerator0
    Uniform Scale Factor=&FloatFromCurve(
     FloatCurve=DataInterface() {propertyValues {Curve {keys [{arriveTangent=2.9879, interpMode=RCIM_Cubic, leaveTangent=2.9879, tangentMode=RCTM_User, time=0, value=0}, <at 1 1 RCIM_Cubic>]}}}
     Scale Curve=&FN_LOD_Distance_Remap(Start Distance=100, End Distance=500, Minimum Output=0.75, Maximum Output=1.25, Clamp Output=true)
    )
   )
   &ScaleColor(
    Scale Alpha=&FloatFromCurve(FloatCurve=DataInterface() {propertyValues {Curve {keys [
         {arriveTangent=9.0725, interpMode=RCIM_Cubic, leaveTangent=9.0725, tangentMode=RCTM_User, time=0, value=0}
         {arriveTangent=-2.5249, interpMode=RCIM_Cubic, leaveTangent=-2.5249, tangentMode=RCTM_User, time=0.1408, value=0.9847}
         {arriveTangent=-0.0405, interpMode=RCIM_Cubic, leaveTangent=-0.0405, tangentMode=RCTM_User, time=1, value=0}
        ]}}})
    Scale RGB=&Multiply_VectorByFloat(
     Vector=&MakeVectorFromLinearColorRGB(Color=$User.Linear Color)
     Float=&FNTimeOfDayScalar(Night Scale=0.2, Morning Scale=0.5, Day Scale=0.5, Evening Scale=0.5)
    )
    ScaleA=true
    ScaleRGB=true
   )
  ]
  rendererClasses [&NiagaraSpriteRendererProperties]
  renderers [{rendererClass=&NiagaraSpriteRendererProperties}]
 }
 FaceAcc {
  emitterUpdateScript [
   &EmitterState()
   &SpawnRate(
    Spawn Probability=$System.SpawnProbability
    SpawnRate=&Multiply_Float(A=$System.SpawnRate, B=1)
    Use Spawn Probability=true
   )
  ]
  particleSpawnScript [
   &PartitionParticles(Particles Per Partition=1, Partition Mode:ENiagara_PartitionMode=::NewEnumerator1)
   &InitializeParticle(
    Color Mode=::NewEnumerator0
    Lifetime=&RandomRangeFloat(Minimum=0.3, Maximum=0.4)
    Sprite Rotation Mode=::NewEnumerator0
    Sprite Size Mode=::NewEnumerator1
    Uniform Sprite Size Max=$System.MaxSize
    Uniform Sprite Size Min=$System.MinSize
    UsePositionOffset=false
    Write Material Random=true
   )
   &CameraOffset(Camera Offset Amount=$System.CameraOffset)
  ]
  particleUpdateScript [
   &ParticleState()
   &SkeletalMeshLocation(
    Bone / Socket Index=&SelectIntFromArray(Direct Array Index=$Particles.PartitionParticles.Partition, Int Selection Array=DataInterface() {propertyValues {IntData [0]}})
    Bone Sampling Mode=::NewEnumerator4
    MaskSpawnGroup=false
    Sampled Position Offset=&SelectVectorFromArray(Array Sampling Mode=::NewEnumerator0, Direct Array Index=?, Vector Selection Array=$User.Array_FaceAcc, Random Seed=0)
    Skeletal Mesh=$User.Skeletal Mesh Face
   ) {module=SkeletalMeshLocation004}
   &SpriteRotationRate(Rotation Rate=250)
   &ScaleSpriteSize(
    Scale Sprite Size Mode=::NewEnumerator0
    Uniform Scale Factor=&FloatFromCurve(
     FloatCurve=DataInterface() {propertyValues {Curve {keys [{arriveTangent=2.9879, interpMode=RCIM_Cubic, leaveTangent=2.9879, tangentMode=RCTM_User, time=0, value=0}, <at 1 1 RCIM_Cubic>]}}}
     Scale Curve=&FN_LOD_Distance_Remap(Start Distance=100, End Distance=500, Minimum Output=0.75, Maximum Output=1.25, Clamp Output=true)
    )
   )
   &ScaleColor(
    Scale Alpha=&FloatFromCurve(FloatCurve=DataInterface() {propertyValues {Curve {keys [
         {arriveTangent=9.0725, interpMode=RCIM_Cubic, leaveTangent=9.0725, tangentMode=RCTM_User, time=0, value=0}
         {arriveTangent=-2.5249, interpMode=RCIM_Cubic, leaveTangent=-2.5249, tangentMode=RCTM_User, time=0.1408, value=0.9847}
         {arriveTangent=-0.0405, interpMode=RCIM_Cubic, leaveTangent=-0.0405, tangentMode=RCTM_User, time=1, value=0}
        ]}}})
    Scale RGB=&Multiply_VectorByFloat(
     Vector=&MakeVectorFromLinearColorRGB(Color=$User.Linear Color)
     Float=&FNTimeOfDayScalar(Night Scale=0.2, Morning Scale=0.5, Day Scale=0.5, Evening Scale=0.5)
    )
    ScaleA=true
    ScaleRGB=true
   )
  ]
  rendererClasses [&NiagaraSpriteRendererProperties]
  renderers [{rendererClass=&NiagaraSpriteRendererProperties}]
 }
 Forearms {
  emitterUpdateScript [
   &EmitterState()
   &SpawnRate(
    Spawn Probability=$System.SpawnProbability
    SpawnRate=&Multiply_Float(A=$System.SpawnRate, B=1)
    Use Spawn Probability=true
   )
  ]
  particleSpawnScript [
   &PartitionParticles(Particles Per Partition=1, Partition Mode:ENiagara_PartitionMode=::NewEnumerator1)
   &InitializeParticle(
    Color Mode=::NewEnumerator0
    Lifetime=&RandomRangeFloat(Minimum=0.3, Maximum=0.4)
    Sprite Rotation Mode=::NewEnumerator0
    Sprite Size Mode=::NewEnumerator1
    Uniform Sprite Size Max=$System.MaxSize
    Uniform Sprite Size Min=$System.MinSize
    UsePositionOffset=false
    Write Material Random=true
   )
   &CameraOffset(Camera Offset Amount=$System.CameraOffset)
  ]
  particleUpdateScript [
   &ParticleState()
   &SkeletalMeshLocation(
    Bone / Socket Index=&SelectIntFromArray(Direct Array Index=$Particles.PartitionParticles.Partition, Int Selection Array=DataInterface() {propertyValues {IntData [6,7]}})
    Bone Sampling Mode=::NewEnumerator4
    MaskSpawnGroup=false
    Sampled Position Offset=&Multiply_Vector(
     A=&SelectVectorFromArray(Array Sampling Mode=::NewEnumerator0, Direct Array Index=?, Vector Selection Array=$User.Array_Forearm, Random Seed=0)
     B=&MaskVectorBySpawnGroup(Masked Vector=<1 1 1>, Passthrough Vector=<-1 -1 -1>, Spawn Group Mask=$Particles.PartitionParticles.Partition)
    )
    Skeletal Mesh=$User.Skeletal Mesh Body
   ) {module=SkeletalMeshLocation004}
   &SpriteRotationRate(Rotation Rate=250)
   &ScaleSpriteSize(
    Scale Sprite Size Mode=::NewEnumerator0
    Uniform Scale Factor=&FloatFromCurve(
     FloatCurve=DataInterface() {propertyValues {Curve {keys [{arriveTangent=2.9879, interpMode=RCIM_Cubic, leaveTangent=2.9879, tangentMode=RCTM_User, time=0, value=0}, <at 1 1 RCIM_Cubic>]}}}
     Scale Curve=&FN_LOD_Distance_Remap(Start Distance=100, End Distance=500, Minimum Output=0.75, Maximum Output=1.25, Clamp Output=true)
    )
   )
   &ScaleColor(
    Scale Alpha=&FloatFromCurve(FloatCurve=DataInterface() {propertyValues {Curve {keys [
         {arriveTangent=9.0725, interpMode=RCIM_Cubic, leaveTangent=9.0725, tangentMode=RCTM_User, time=0, value=0}
         {arriveTangent=-2.5249, interpMode=RCIM_Cubic, leaveTangent=-2.5249, tangentMode=RCTM_User, time=0.1408, value=0.9847}
         {arriveTangent=-0.0405, interpMode=RCIM_Cubic, leaveTangent=-0.0405, tangentMode=RCTM_User, time=1, value=0}
        ]}}})
    Scale RGB=&Multiply_VectorByFloat(
     Vector=&MakeVectorFromLinearColorRGB(Color=$User.Linear Color)
     Float=&FNTimeOfDayScalar(Night Scale=0.2, Morning Scale=0.5, Day Scale=0.5, Evening Scale=0.5)
    )
    ScaleA=true
    ScaleRGB=true
   )
  ]
  rendererClasses [&NiagaraSpriteRendererProperties]
  renderers [{rendererClass=&NiagaraSpriteRendererProperties}]
 }
 Thighs {
  emitterUpdateScript [
   &EmitterState()
   &SpawnRate(
    Spawn Probability=$System.SpawnProbability
    SpawnRate=&Multiply_Float(A=$System.SpawnRate, B=1)
    Use Spawn Probability=true
   )
  ]
  particleSpawnScript [
   &PartitionParticles(Particles Per Partition=1, Partition Mode:ENiagara_PartitionMode=::NewEnumerator1)
   &InitializeParticle(
    Color Mode=::NewEnumerator0
    Lifetime=&RandomRangeFloat(Minimum=0.3, Maximum=0.4)
    Sprite Rotation Mode=::NewEnumerator0
    Sprite Size Mode=::NewEnumerator1
    Uniform Sprite Size Max=$System.MaxSize
    Uniform Sprite Size Min=$System.MinSize
    UsePositionOffset=false
    Write Material Random=true
   )
   &CameraOffset(Camera Offset Amount=$System.CameraOffset)
  ]
  particleUpdateScript [
   &ParticleState()
   &SkeletalMeshLocation(
    Bone / Socket Index=&SelectIntFromArray(Direct Array Index=$Particles.PartitionParticles.Partition, Int Selection Array=DataInterface() {propertyValues {IntData [2,3]}})
    Bone Sampling Mode=::NewEnumerator4
    MaskSpawnGroup=false
    Sampled Position Offset=&Multiply_Vector(
     A=&SelectVectorFromArray(Array Sampling Mode=::NewEnumerator0, Direct Array Index=?, Vector Selection Array=$User.Array_Thigh, Random Seed=0)
     B=&MaskVectorBySpawnGroup(Masked Vector=<1 1 1>, Passthrough Vector=<-1 -1 -1>, Spawn Group Mask=$Particles.PartitionParticles.Partition)
    )
    Skeletal Mesh=$User.Skeletal Mesh Body
   ) {module=SkeletalMeshLocation004}
   &SpriteRotationRate(Rotation Rate=250)
   &ScaleSpriteSize(
    Scale Sprite Size Mode=::NewEnumerator0
    Uniform Scale Factor=&FloatFromCurve(
     FloatCurve=DataInterface() {propertyValues {Curve {keys [{arriveTangent=2.9879, interpMode=RCIM_Cubic, leaveTangent=2.9879, tangentMode=RCTM_User, time=0, value=0}, <at 1 1 RCIM_Cubic>]}}}
     Scale Curve=&FN_LOD_Distance_Remap(Start Distance=100, End Distance=500, Minimum Output=0.75, Maximum Output=1.25, Clamp Output=true)
    )
   )
   &ScaleColor(
    Scale Alpha=&FloatFromCurve(FloatCurve=DataInterface() {propertyValues {Curve {keys [
         {arriveTangent=9.0725, interpMode=RCIM_Cubic, leaveTangent=9.0725, tangentMode=RCTM_User, time=0, value=0}
         {arriveTangent=-2.5249, interpMode=RCIM_Cubic, leaveTangent=-2.5249, tangentMode=RCTM_User, time=0.1408, value=0.9847}
         {arriveTangent=-0.0405, interpMode=RCIM_Cubic, leaveTangent=-0.0405, tangentMode=RCTM_User, time=1, value=0}
        ]}}})
    Scale RGB=&Multiply_VectorByFloat(
     Vector=&MakeVectorFromLinearColorRGB(Color=$User.Linear Color)
     Float=&FNTimeOfDayScalar(Night Scale=0.2, Morning Scale=0.5, Day Scale=0.5, Evening Scale=0.5)
    )
    ScaleA=true
    ScaleRGB=true
   )
  ]
  rendererClasses [&NiagaraSpriteRendererProperties]
  renderers [{rendererClass=&NiagaraSpriteRendererProperties}]
 }
}
summary {
 systemName=NS_MED_SkeleProbe_Candy_Glitter_
 emitters [
  {emitterName=Calfs, rendererClasses [&NiagaraSpriteRendererProperties]}
  {emitterName=Arms, rendererClasses [&NiagaraSpriteRendererProperties]}
  {emitterName=Thighs, rendererClasses [&NiagaraSpriteRendererProperties]}
  {emitterName=Forearms, rendererClasses [&NiagaraSpriteRendererProperties]}
  {emitterName=Belly_Back, rendererClasses [&NiagaraSpriteRendererProperties]}
  {emitterName=Chest, rendererClasses [&NiagaraSpriteRendererProperties]}
  {emitterName=FaceAcc, rendererClasses [&NiagaraSpriteRendererProperties]}
 ]
 userVariables [
  User.Array_Arm {
   type=NiagaraDataInterfaceArrayFloat3
   description=""
   defaultValue {
    struct=NiagaraExt_VariableValue_DataInterface
    value {
     dataInterface=&self:NiagaraDataInterfaceArrayFloat3_2
     dataInterfaceClass=&NiagaraDataInterfaceArrayFloat3
    }
   }
  }
  User.Array_Belly {
   type=NiagaraDataInterfaceArrayFloat3
   description=""
   defaultValue {
    struct=NiagaraExt_VariableValue_DataInterface
    value {
     dataInterface=&self:NiagaraDataInterfaceArrayFloat3_4
     dataInterfaceClass=&NiagaraDataInterfaceArrayFloat3
    }
   }
  }
  User.Array_Calf {
   type=NiagaraDataInterfaceArrayFloat3
   description=""
   defaultValue {
    struct=NiagaraExt_VariableValue_DataInterface
    value {
     dataInterface=&self:NiagaraDataInterfaceArrayFloat3_0
     dataInterfaceClass=&NiagaraDataInterfaceArrayFloat3
    }
   }
  }
  User.Array_Chest {
   type=NiagaraDataInterfaceArrayFloat3
   description=""
   defaultValue {
    struct=NiagaraExt_VariableValue_DataInterface
    value {
     dataInterface=&self:NiagaraDataInterfaceArrayFloat3_5
     dataInterfaceClass=&NiagaraDataInterfaceArrayFloat3
    }
   }
  }
  User.Array_FaceAcc {
   type=NiagaraDataInterfaceArrayFloat3
   description=""
   defaultValue {
    struct=NiagaraExt_VariableValue_DataInterface
    value {
     dataInterface=&self:NiagaraDataInterfaceArrayFloat3_6
     dataInterfaceClass=&NiagaraDataInterfaceArrayFloat3
    }
   }
  }
  User.Array_Forearm {
   type=NiagaraDataInterfaceArrayFloat3
   description=""
   defaultValue {
    struct=NiagaraExt_VariableValue_DataInterface
    value {
     dataInterface=&self:NiagaraDataInterfaceArrayFloat3_3
     dataInterfaceClass=&NiagaraDataInterfaceArrayFloat3
    }
   }
  }
  User.Array_Thigh {
   type=NiagaraDataInterfaceArrayFloat3
   description=""
   defaultValue {
    struct=NiagaraExt_VariableValue_DataInterface
    value {
     dataInterface=&self:NiagaraDataInterfaceArrayFloat3_1
     dataInterfaceClass=&NiagaraDataInterfaceArrayFloat3
    }
   }
  }
  User.Linear Color {type=LinearColor, description="", defaultValue {struct=LinearColor, value=<rgba 1 0.175 0.7938 1>}}
  User.Skeletal Mesh Body {
   type=NiagaraDataInterfaceSkeletalMesh
   defaultValue {
    struct=NiagaraExt_VariableValue_DataInterface
    value {dataInterface=&self:NiagaraDataInterfaceSkeletalMesh_2}
   }
  }
  User.Skeletal Mesh Face {
   type=NiagaraDataInterfaceSkeletalMesh
   defaultValue {
    struct=NiagaraExt_VariableValue_DataInterface
    value {dataInterface=&self:NiagaraDataInterfaceSkeletalMesh_1}
   }
  }
 ]
}
